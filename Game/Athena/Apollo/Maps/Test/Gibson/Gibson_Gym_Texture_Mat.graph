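# Test material for the Gibson gym test map under Athena/Apollo, using the Default Lit shading model.
# The graph holds a single Texture Sample node that references Gibson_Gym_Texture.
# No connection from the sample to a material output is shown, and the graph has no parameters or other logic.
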Material /Game/Athena/Apollo/Maps/Test/Gibson/Gibson_Gym_Texture_Mat
graph {
 properties {pixelDepthOffsetMode=PDOM_Legacy, usedShadingModels=Default Lit}
 nodes [TextureSample() {texture=&/Gibson_Gym_Texture}]
}
